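# M_Edison_Decal is a masked, two-sided, Default Lit material for decals on the Edison vehicle body, enabled for skeletal meshes and mesh deformers.
# It samples dedicated base color, mask and normal textures.
# The base color texture's alpha is inverted and clamped with a raised lower limit, so that channel carries packed surface data.
# The file does not show which material inputs these nodes drive.
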Material /VehicleCosmetics/Bodies/Edison/Materials/M_Edison_Decal
graph {
 properties {
  bAutomaticallySetUsageInEditor=false
  bUsedWithMeshDeformer=true
  bUsedWithSkeletalMesh=true
  blendMode=BLEND_Masked
  twoSided=true
  usedShadingModels=Default Lit
 }
 nodes [
  Clamp(None=OneMinus(None=TextureSample() {texture=/VehicleCosmetics/Bodies/Edison/Textures/T_Edison_Decal_BC}.A)) {minDefault=0.2}
  TextureSample() {texture=/VehicleCosmetics/Bodies/Edison/Textures/T_Edison_Decal_Mask}
  TextureSample() {samplerType=SAMPLERTYPE_Normal, texture=/VehicleCosmetics/Bodies/Edison/Textures/T_Edison_Decal_N}
 ]
}
